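# Sound cue for the Creative Speed Trap's firing sound, routed to the trap placement sound class.
# A quality-level switch feeds four pitch-modulated branches: two randomly pick between Fire02 and Fire03 waves, while the other two play only Fire02.
# The cue plays quieter and higher-pitched than the source waves, using shared trap attenuation.
# Overridden concurrency allows two instances per owner, stopping the oldest.
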SoundCue /Game/Sounds/Fort_Traps/Speed_Trap/Speed_Trap_Creative_Fire_Cue
root=QualityLevel_0
nodes [
 {id=Modulator_0, children [WavePlayer_1], properties {pitchMax=1.1, pitchMin=0.9}}
 {id=Modulator_1, children [Random_0], properties {pitchMax=1.1, pitchMin=0.9}}
 {id=Modulator_2, children [WavePlayer_4], properties {pitchMax=1.1, pitchMin=0.9}}
 {id=Modulator_6, children [Random_13], properties {pitchMax=1.1, pitchMin=0.9}}
 {id=QualityLevel_0, children [Modulator_6,Modulator_1,Modulator_0,Modulator_2]}
 {id=Random_0, children [WavePlayer_3,WavePlayer_6], properties {weights [1,1]}}
 {id=Random_13, children [WavePlayer_53, WavePlayer_54], properties {weights [1,1]}}
 {id=WavePlayer_1, properties {soundWaveAssetPtr=&/Speed_Trap_Fire02}}
 {id=WavePlayer_3, properties {soundWaveAssetPtr=&Speed_Trap_Fire02}}
 {id=WavePlayer_4, properties {soundWaveAssetPtr=&Speed_Trap_Fire02}}
 {id=WavePlayer_53, properties {soundWaveAssetPtr=&Speed_Trap_Fire02}}
 {id=WavePlayer_54, properties {soundWaveAssetPtr=&/Speed_Trap_Fire03}}
 {id=WavePlayer_6, properties {soundWaveAssetPtr=&Speed_Trap_Fire03}}
]
properties {
 attenuationSettings=&Fort_Traps_Att
 bOverrideConcurrency=true
 maxDistance=6000
 pitchMultiplier=1.5
 soundClassObject=/Game/Sounds/FortSoundClassesAndModes/Fort_Traps_Placement
 volumeMultiplier=0.2
 attenuationOverrides {bAttenuateWithLPF=true, falloffDistance=3100, lPFRadiusMin=2500, attenuationShapeExtents {x=900}}
 concurrencyOverrides {bLimitToOwner=true, maxCount=2, resolutionRule=StopOldest}
}
