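# Sound cue for placing a Speed Trap, routed to the trap placement sound class with Fort_Traps attenuation.
# A root quality-level node selects among four branches: random of three placement waves, random of two, and two identical single-wave branches.
# The single-wave branches use slight pitch modulation for variation in place of alternate samples.
# Overridden concurrency prevents new instances from starting once the limit is reached.
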SoundCue /Game/Sounds/Fort_Traps/Speed_Trap/Speed_Trap_Place_Cue
root=QualityLevel_0
nodes [
 {id=Modulator_0, children [WavePlayer_0], properties {pitchMax=1.08, pitchMin=0.92}}
 {id=Modulator_2, children [WavePlayer_5], properties {pitchMax=1.08, pitchMin=0.92}}
 {id=QualityLevel_0, children [Random_14,Random_0,Modulator_0,Modulator_2]}
 {id=Random_0, children [WavePlayer_3,WavePlayer_1], properties {weights [1,1]}}
 {id=Random_14, children [WavePlayer_55, WavePlayer_56, WavePlayer_4], properties {weights [1,1,1]}}
 {id=WavePlayer_0, properties {soundWaveAssetPtr=&/Speed_Trap_Place01}}
 {id=WavePlayer_1, properties {soundWaveAssetPtr=&/Speed_Trap_Place02}}
 {id=WavePlayer_3, properties {soundWaveAssetPtr=&Speed_Trap_Place01}}
 {id=WavePlayer_4, properties {soundWaveAssetPtr=&/Speed_Trap_Place03}}
 {id=WavePlayer_5, properties {soundWaveAssetPtr=&Speed_Trap_Place01}}
 {id=WavePlayer_55, properties {soundWaveAssetPtr=&Speed_Trap_Place01}}
 {id=WavePlayer_56, properties {soundWaveAssetPtr=&Speed_Trap_Place02}}
]
properties {
 attenuationSettings=&Fort_Traps_Att
 bOverrideConcurrency=true
 maxDistance=6000
 soundClassObject=/Game/Sounds/FortSoundClassesAndModes/Fort_Traps_Placement
 volumeMultiplier=0.5
 attenuationOverrides {falloffDistance=2400, attenuationShapeExtents {x=600}}
 concurrencyOverrides {resolutionRule=PreventNew}
}
