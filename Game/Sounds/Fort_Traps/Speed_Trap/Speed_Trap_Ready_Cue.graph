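# Sound cue for the speed trap's ready sound, in the Fort_Traps sound class.
# A Random node picks between two equally weighted Ready wave variants, then a Modulator varies the pitch on playback.
# The cue plays at reduced volume with the shared trap attenuation settings and overrides concurrency with a rule that prevents new instances at the limit.
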SoundCue /Game/Sounds/Fort_Traps/Speed_Trap/Speed_Trap_Ready_Cue
root=Modulator_10
nodes [
 {id=Modulator_10, children [Random_20], properties {pitchMax=1.1}}
 {id=Random_20, children [WavePlayer_71, WavePlayer_72], properties {weights [1,1]}}
 {id=WavePlayer_71, properties {soundWaveAssetPtr=&/Speed_Trap_Ready02}}
 {id=WavePlayer_72, properties {soundWaveAssetPtr=&/Speed_Trap_Ready03}}
]
properties {
 attenuationSettings=&Fort_Traps_Att
 bOverrideConcurrency=true
 maxDistance=6000
 soundClassObject=&Fort_Traps
 volumeMultiplier=0.5
 attenuationOverrides {falloffDistance=2100}
 concurrencyOverrides {resolutionRule=PreventNew}
}
